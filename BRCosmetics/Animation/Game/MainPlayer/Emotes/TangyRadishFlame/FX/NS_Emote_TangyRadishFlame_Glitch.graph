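# Niagara system for the glitch effect of the TangyRadishFlame emote on the main player.
# It contains two glitch box emitters, GlitchBox and GlitchBox001, with no renderers listed here.
# A float user parameter, SpawnSwitch, is exposed, and system spawn runs several variable assignments alongside SystemState.
# Dynamic inputs include float-comparison booleans, bool-to-int and bool-to-float conversions, random color ranges and time-of-day color.
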NiagaraSystem /BRCosmetics/Animation/Game/MainPlayer/Emotes/TangyRadishFlame/FX/NS_Emote_TangyRadishFlame_Glitch
dependencies {
 usedDataInterfaces []
 usedDynamicInputs [
  &SetBoolByFloatComparison
  &MakeCustomIntFromBool
  &MakeCustomFloatFromBool
  &Multiply_VectorByFloat
  &FNTimeOfDayColor
  &RandomRangeLinearColor
 ]
 usedModules [
  &self:SystemSpawnScript.SystemScriptSource.SystemScriptGraph.NiagaraNodeAssignment_0.SetVariables_#1
  &self:SystemSpawnScript.SystemScriptSource.SystemScriptGraph.NiagaraNodeAssignment_1.SetVariables_#2
  &self:SystemSpawnScript.SystemScriptSource.SystemScriptGraph.NiagaraNodeAssignment_2.SetVariables_#3
  &self:SystemSpawnScript.SystemScriptSource.SystemScriptGraph.NiagaraNodeAssignment_3.SetVariables_#4
  &SystemState
 ]
}
emitters {GlitchBox {rendererClasses [], renderers []}, GlitchBox001 {rendererClasses [], renderers []}}
summary {
 systemName=NS_Emote_TangyRadishFlame_Glitch
 emitters [{emitterName=GlitchBox, rendererClasses []}, {emitterName=GlitchBox001, rendererClasses []}]
 userVariables [User.SpawnSwitch {type=NiagaraFloat, defaultValue=1, description=""}]
}
