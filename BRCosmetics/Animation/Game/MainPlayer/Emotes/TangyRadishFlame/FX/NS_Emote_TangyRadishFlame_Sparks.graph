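# Niagara spark effect for the TangyRadishFlame emote, attached to two sockets sampled through a shared Socket Reader.
# Classic_Sparks emits flickering, speed-stretched falling sparks, while Flash_Flare adds socket-following flare sprites and lights; both stop spawning around 4.1 seconds.
# ParticleDecal keeps glow sprites and meshes on the sockets, blending to faster alpha flicker when smoothed owner velocity crosses a threshold.
# Spawn-rate lerps also read that speed blend, but their endpoints are identical.
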NiagaraSystem /BRCosmetics/Animation/Game/MainPlayer/Emotes/TangyRadishFlame/FX/NS_Emote_TangyRadishFlame_Sparks
dependencies {
 usedDataInterfaces [
  &NiagaraDataInterfaceCurve
  &NiagaraDataInterfaceArrayInt32
  &NiagaraDataInterfaceVector2DCurve
  &NiagaraDataInterfaceSpriteRendererInfo
  &NiagaraDataInterfaceColorCurve
  /Script/Niagara.NiagaraDataInterfaceArrayQuat
  &NiagaraDataInterfaceSocketReader
 ]
 usedDynamicInputs []
 usedModules [
  &self:SystemSpawnScript.SystemScriptSource.SystemScriptGraph.NiagaraNodeAssignment_1.SetVariables_#1
  &SystemState
  &BlendSpaceHelper
  &self:SystemSpawnScript.SystemScriptSource.SystemScriptGraph.NiagaraNodeAssignment_0.SetVariables_#2
 ]
}
emitters {
 Classic_Sparks {
  emitterUpdateScript [
   &self:InterpolateSpeedThresholdEmitter_01(
    Define Initial Value=false
    Rate Of Change=5
    Target Float Value:NiagaraFloat=&MakeCustomFloatFromBool(
     Bool=&SetBoolByFloatComparison(
      A=&Add_Float(
       A=&Add_Float(
        A=&Abs_Float(Float:NiagaraFloat=&MakeFloatFromVector(Vector=$Engine.Owner.Velocity, Channel=::NewEnumerator0))
        B=&Abs_Float(Float:NiagaraFloat=&MakeFloatFromVector(Vector=$Engine.Owner.Velocity, Channel=::NewEnumerator2))
       )
       B=&Abs_Float(Float:NiagaraFloat=&MakeFloatFromVector(Vector=$Engine.Owner.Velocity, Channel=::NewEnumerator1))
      )
      B=46.278
     )
     True Float=1
     False Float=0
    )
   ) {module=InterpolateSpeedThresholdEmitter_01}
   &EmitterState(Life Cycle Mode=::NewEnumerator1, Loop Delay=2.3833, Loop Duration=2, UseLoopDelay=false)
   &SpawnRate(
    SpawnRate=&Lerp_Float(
     A=&FloatFromCurve(FloatCurve=DataInterface() {propertyValues {Curve {keys [<at 4.1 1>,<at 4.2 0>]}}}, CurveIndex=$System.Age, Scale Curve=300)
     B=&FloatFromCurve(FloatCurve=DataInterface() {propertyValues {Curve {keys [<at 4.1 1>,<at 4.2 0>]}}}, CurveIndex=$System.Age, Scale Curve=300)
     Alpha=$Emitter.InterpolateSpeedThresholdEmitter_01.InterpolateOverTime_Float.Moving Average
    )
    Use Spawn Probability=true
   )
  ]
  particleSpawnScript [
   &PartitionParticles()
   &InitializeParticle(
    Color Mode=::NewEnumerator0
    Lifetime=&MaskFloatBySpawnGroup(
     Masked Float=&RandomRangeFloat(Minimum=0.25, Maximum=0.18)
     Passthrough Float=&RandomRangeFloat(Minimum=0.1, Maximum=0.2)
    )
    Mesh Scale Max=<0.0663 0.0663 0.0663>
    Mesh Scale Min=<2.4885 2.8084 2.6334>
    Mesh Scale Mode=::NewEnumerator2
    Position Mode=::NewEnumerator0
    Sprite Rotation Mode=::NewEnumerator2
    Sprite Size Max=<3 6>
    Sprite Size Min=<1 1>
    Sprite Size Mode=::NewEnumerator2
    Write Material Random=true
   )
   &SocketLocation(
    "Inherit Velocity (Scale)"=0
    SamplePosOffset=false
    Sampled Velocity Coordinate Space:ENiagaraCoordinateSpace=::Simulation
    Socket Index=&SelectIntFromArray(Direct Array Index=$Particles.PartitionParticles.Partition, Int Selection Array=DataInterface() {propertyValues {IntData [0,1]}})
    Socket Reader=$System.Socket Reader
    Socket Sampling Mode=::NewEnumerator3
    VelocitySpeedLimit=false
   )
   &AddVelocity(
    Cone Axis=<0 -1 0.85>
    Module.UseConeFalloff=false
    Module.UseVelDistribution=false
    Override Seed=false
    Velocity Mode=::NewEnumerator2
    Velocity Speed=&MaskFloatBySpawnGroup(Masked Float=150, Passthrough Float=0)
   ) {module=AddVelocity001}
   &ShapeLocation(
    Box Size=<9.1912 2 0.5>
    MaskSpawnGroup=false
    Override Seed=false
    Rotation Coordinate Space=::World
    Shape Primitive=::NewEnumerator2
    Surface Only=false
    UseOwnerScale=true
   )
   &ShapeLocation(
    Box Size=<60 60 150>
    MaskSpawnGroup=true
    Offset=<0 0 58.7788>
    Offset Mode=::NewEnumerator0
    Override Seed=false
    Shape Primitive=::NewEnumerator2
    Spawn Group Mask=1
    Surface Only=false
    UseOwnerScale=true
   ) {module=ShapeLocation001}
  ]
  particleUpdateScript [
   &ParticleState()
   /SetVariables(
    Particles.Flicker:NiagaraFloat=&Waveform("Waveform [1]"=::NewEnumerator0, "[1] Frequency"=75, "[1] Amplitude Min/Max"=<0 2>)
    Particles.Flicker Size:NiagaraFloat=&Waveform("Waveform [1]"=::NewEnumerator0, "[1] Frequency"=75, "[1] Amplitude Min/Max"=<0.5 1>)
    Particles.Flicker Value:NiagaraFloat=&Waveform("Waveform [1]"=::NewEnumerator0, "[1] Frequency"=75, "[1] Amplitude Min/Max"=<2 8>)
   ) {isSetParameters=true}
   &ScaleSpriteSize(
    Scale Sprite Size Mode=::NewEnumerator0
    Uniform Scale Factor=&Lerp_Float(
     A=1
     B=&Multiply_Float(A=$Particles.Flicker Size, B=1)
     Alpha=&FloatFromCurve(FloatCurve=DataInterface() {propertyValues {Curve {keys [<at 0 0.1076>,<at 0.2485 0.6128>,<at 0.45 1>,<at 1 1>]}}})
    )
   )
   &ScaleColor(
    Scale Mode=::NewEnumerator1
    Scale RGBA=&Multiply_Vector4(
     A=&MakeVector4_FromLinearColor(LinearColor=&ColorFromCurve(ColorCurve=DataInterface() {propertyValues {
         AlphaCurve {keys [<at 0.122 1>]}
         BlueCurve {keys [<at 0 1.0417>,<at 0.1355 0.9911>,<at 0.3032 0>,<at 1 0>]}
         GreenCurve {keys [<at 0 7.4334>,<at 0.1355 10.4768>,<at 0.3032 9.1355>,<at 1 3.6718>]}
         RedCurve {keys [<at 0 20>,<at 0.1355 16>,<at 0.3032 13>,<at 1 9>]}
        }}))
     B=&MakeVector4_FromLinearColor(LinearColor=&MakeLinearColorFromVectorAndFloat(
       "Vector (RGB)"=&Add_Vector(
        A=<1 1 1>
        B=/Niagara/DynamicInputs/TypeConversions/MakeVectorFromLinearColorHSV(Color:LinearColor=/Niagara/DynamicInputs/LinearColor/MakeLinearColorFromHSV(
          HSV From:NiagaraInt32=1
          HSV:Vector3f=?
          Hue:NiagaraFloat=0
          Saturation:NiagaraFloat=0
          Value:NiagaraFloat=&Lerp_Float(
           A=0
           B=$Particles.Flicker Value
           Alpha=&FloatFromCurve(FloatCurve=DataInterface() {propertyValues {Curve {keys [<at 0 0>,<at 0.5 0>,<at 0.55 0.9624>,<at 1 1>]}}})
          )
          Alpha:NiagaraFloat=1
         ))
       )
       "Float (Alpha)"=&Lerp_Float(
        A=1
        B=$Particles.Flicker
        Alpha=&FloatFromCurve(FloatCurve=DataInterface() {propertyValues {Curve {keys [<at 0 0>,<at 0.25 0>,<at 0.35 1>,<at 1 1>]}}})
       )
      ))
    )
    ScaleRGBA=true
   )
   &GravityForce(Gravity=&RandomRangeVector(Minimum=<0 0 -161.3829>, Maximum=<0 0 -457.3416>))
   &SolveForcesAndVelocity(Clamp Velocity=false, Limit Acceleration=false)
   &ScaleSpriteSizeBySpeed(
    Max Scale Factor=&RandomRangeVector2D(Minimum=<0.22 18.0452>, Maximum=<6 11.7483>)
    Min Scale Factor=<0.15 1>
    Velocity Threshold=3800
   )
   /SetVariables() {isSetParameters=true}
  ]
  rendererClasses [&NiagaraSpriteRendererProperties]
  renderers [{rendererClass=&NiagaraSpriteRendererProperties}]
 }
 Flash_Flare {
  emitterUpdateScript [
   &EmitterState(Inactive Response=::NewEnumerator2, Life Cycle Mode=::NewEnumerator1, Loop Behavior=::NewEnumerator1, Loop Duration=4.1, UseLoopDelay=false)
   &SpawnBurst_Instantaneous(Spawn Count=2, Spawn Group=2, Use Spawn Probability=false, UseLoopCountLimit=false)
   &SpawnRate(
    SpawnRate=&Lerp_Float(
     A=&FloatFromCurve(FloatCurve=DataInterface() {propertyValues {Curve {keys [<at 4.1 1>,<at 4.2 0>]}}}, CurveIndex=$System.Age, Scale Curve=8)
     B=&FloatFromCurve(FloatCurve=DataInterface() {propertyValues {Curve {keys [<at 4.1 1>,<at 4.2 0>]}}}, CurveIndex=$System.Age, Scale Curve=8)
     Alpha=$Emitter.InterpolateSpeedThresholdEmitter_02.InterpolateOverTime_Float.Moving Average
    )
    Use Spawn Probability=false
   )
  ]
  particleSpawnScript [
   /SetVariables(Particles.VisibilityTag=&SpawnGroup()) {isSetParameters=true}
   &PartitionParticles()
   &InitializeParticle(
    Color=<rgba 1.5 0.841 0.4297 1>
    Lifetime=&MaskFloatBySpawnGroup(Masked Float=0.36, Passthrough Float=0.36)
    Position=&SimulationPosition()
    Position Mode=::NewEnumerator1
    Position Offset=&MaskVectorBySpawnGroup(Masked Vector=<0 0 -16>, Passthrough Vector=<0 0 0>)
    Position Offset Coordinate Space=::World
    Sprite Size=<130 72>
    Sprite Size Mode=::NewEnumerator4
    UsePositionOffset=true
    Write Material Random=true
   ) {module=InitializeParticle001}
   &DynamicMaterialParameters(
    Index 0 Param 1=0.2
    Module.Param0WriteEnabled=true
    Module.Param1WriteEnabled=true
    Module.Param2WriteEnabled=true
    Module.Param3WriteEnabled=true
   )
   &CameraOffset(Camera Offset Amount=6) {module=CameraOffset001}
  ]
  particleUpdateScript [
   &ParticleState()
   &SolveForcesAndVelocity(Clamp Velocity=false, Limit Acceleration=false)
   &SocketLocation(
    Position Sampling=::NewEnumerator1
    SamplePosOffset=false
    Socket Index=&SelectIntFromArray(Direct Array Index=$Particles.PartitionParticles.Partition, Int Selection Array=DataInterface() {propertyValues {IntData [0,1]}})
    Socket Reader=$System.Socket Reader
    Socket Sampling Mode=::NewEnumerator3
    VelocitySpeedLimit=false
   )
   /SetVariables(
    Particles.SparkColor:LinearColor=<rgba 5 1.962 0.0658 1>
    Particles.SparkSize:Vector2f=&MakeVector2D(
     X=29.5912
     Y=&ClampFloat(Float=&Sine(Normalized Angle=$Emitter.Age, Period=0.0616, Scale=4.1323, Bias=4.3498), Min=12.5264, Max=24.152)
    )
    Particles.Vector 2D:Vector2f=&Vector2DFromCurve(
     Vector2Curve=DataInterface() {propertyValues {
       XCurve {defaultValue=1, keys [{tangentMode=RCTM_User, time=0, value=0.5}, {tangentMode=RCTM_User, time=1, value=1}]}
       YCurve {defaultValue=1, keys [{tangentMode=RCTM_User, time=0, value=0.5}, {tangentMode=RCTM_User, time=1, value=1}]}
      }}
     ScaleCurve=&RandomRangeVector2D(Minimum=<18 6.7222>, Maximum=<18 6.4088>)
    )
   ) {isSetParameters=true}
   &SubUVAnimation(
    Random Start Frame=true
    "Sprite Renderer "=DataInterface() {propertyValues {SpriteRenderer=&self:Flash_Flare_0.NiagaraSpriteRendererProperties_1}}
    SubUV Lookup Index Scale=5.7547
    UseEndFrame=false
    UseStartFrame=false
   )
   /SetVariables(Particles.Position=&ConvertPositionToVector(
     Input Position=&AddVectorToPosition(
      Position=$Output.SocketLocation.SampledPosition
      Vector=&MaskVectorBySpawnGroup(Masked Vector=<0 1 -2>, Passthrough Vector=<0 1 0>)
     )
     Conversion Operation:ENiagara_LWCConvertPosToVec=::NewEnumerator0
    )) {isSetParameters=true}
   &FN_Light_Attributes(
    Color=<rgba 2 1.0278 0 1>
    Create Color Attribute=true
    Create Radius Attribute=true
    Radius=70
    Scale Color=&Multiply_Float(A=1.5, B=$System.FlickerSpeed)
   )
   /SetVariables(Particles.PositionLights:NiagaraPosition=&AddVectorToPosition(Position=$Output.SocketLocation.SampledPosition, Vector=<0 -20 10>)) {isSetParameters=true}
  ]
  rendererClasses [&NiagaraSpriteRendererProperties, &NiagaraLightRendererProperties]
  renderers [
   {rendererClass=&NiagaraSpriteRendererProperties}
   {rendererClass=&NiagaraLightRendererProperties, rendererIndex=1}
   {rendererClass=&NiagaraSpriteRendererProperties, rendererIndex=2}
  ]
 }
 ParticleDecal {
  emitterUpdateScript [
   &EmitterState(Life Cycle Mode=::NewEnumerator1, Loop Behavior=::NewEnumerator1, Loop Duration=1, UseLoopDelay=false)
   &SpawnBurst_Instantaneous(Spawn Count=2, Use Spawn Probability=false, UseLoopCountLimit=false)
   &SpawnBurst_Instantaneous(Spawn Count=2, Spawn Group=1, Use Spawn Probability=false, UseLoopCountLimit=false) {module=SpawnBurst_Instantaneous001}
  ]
  particleSpawnScript [
   /SetVariables(Particles.VisibilityTag=&SpawnGroup()) {isSetParameters=true}
   &InitializeParticle(
    Color=&MergeHelper_ColorSwitch(
     0=<rgba 1 0.4751 0.0454 0.6348>
     1=<rgba 25 7.3854 1.0992 0.3652>
     2=?
     3=?
     Index=&SpawnGroup()
     Num=::NewEnumerator0
    )
    Lifetime=4.3
    Mesh Scale=<0.03 0.025 0.0315>
    Mesh Scale Mode=::NewEnumerator4
    Position Offset=<0 0 376.2355>
    Sprite Size Mode=::NewEnumerator3
    Uniform Sprite Size=12
    UsePositionOffset=true
    Write Material Random=true
   )
   &CameraOffset(Camera Offset Amount=&MaskFloatBySpawnGroup(Masked Float=15, Passthrough Float=2))
   &DynamicMaterialParameters(
    Index 0 Param 1=1
    Index 0 Param 3=1
    Module.Index0Param0Write=true
    Module.Index0Param1Write=true
    Module.Index0Param2Write=true
    Module.Index0Param3Write=true
   )
   &PartitionParticles()
   &InitialMeshOrientation(
    Mesh Orientation Mode=::NewEnumerator6
    Orientation Quaternion:Quat4f=/Niagara/DynamicInputs/Arrays/SelectQuaternionFromArray(
     Array Sampling Mode:ENiagaraArraySamplingMode=::NewEnumerator1
     Direct Array Index:NiagaraInt32=$Particles.PartitionParticles.Partition
     Interpolation:NiagaraFloat=?
     Direct Array Mode:ENiagara_WrapClamp=::NewEnumerator0
     Evaluation Type:ENiagaraRandomnessEvaluation=::NewEnumerator0
     Normalized Interpolation:NiagaraBool=true
     Recalculate Random Each Loop:NiagaraBool=false
     Randomness Mode:ENiagaraRandomnessMode=::NewEnumerator0
     Random Seed:NiagaraInt32=?
     Fixed Random Seed:NiagaraBool=false
     Quaternion Selection Array:NiagaraDataInterfaceArrayQuat=DataInterface() {propertyValues {GpuSyncMode=SyncCpuToGpu, MaxElements=0, QuatData [<0 -0.7071 0 0.7071>, <0 0.7071 0 0.7071>]}}
     Enable Spherical Linear Interpolation:NiagaraBool=?
    )
    Use Rotation Vector=false
   )
  ]
  particleUpdateScript [
   &ParticleState()
   &self:InterpolateSpeedThresholdEmitter_01(
    Define Initial Value=false
    Rate Of Change=12
    Target Float Value:NiagaraFloat=&MakeCustomFloatFromBool(
     Bool=&SetBoolByFloatComparison(
      A=&Add_Float(
       A=&Add_Float(
        A=&Abs_Float(Float:NiagaraFloat=&MakeFloatFromVector(Vector=$Engine.Owner.Velocity, Channel=::NewEnumerator0))
        B=&Abs_Float(Float:NiagaraFloat=&MakeFloatFromVector(Vector=$Engine.Owner.Velocity, Channel=::NewEnumerator2))
       )
       B=&Abs_Float(Float:NiagaraFloat=&MakeFloatFromVector(Vector=$Engine.Owner.Velocity, Channel=::NewEnumerator1))
      )
      B=46.278
     )
     True Float=1
     False Float=0
    )
   ) {module=InterpolateSpeedThresholdEmitter_01}
   &SocketLocation(
    SamplePosOffset=true
    Sampled Position Offset=&MaskVectorBySpawnGroup(Masked Vector=<0 1 0>, Passthrough Vector=<0 1.5 7>)
    Sampled Position Offset Coordinate Space:ENiagaraSkelMeshTransforms=::NewEnumerator2
    Socket Index=&SelectIntFromArray(Direct Array Index=$Particles.PartitionParticles.Partition, Int Selection Array=DataInterface() {propertyValues {IntData [0,1]}})
    Socket Reader=$System.Socket Reader
    Socket Sampling Mode=::NewEnumerator3
    VelocitySpeedLimit=false
   )
   &ScaleColor(
    Scale Alpha=&Lerp_Float(
     A=&Multiply_Float(
      A=&FloatFromCurve(FloatCurve=DataInterface() {propertyValues {Curve {keys [<at 0.9323 1>,<at 1 0>]}}})
      B=&Waveform("Waveform [1]"=::NewEnumerator0, "[1] Frequency"=4.1057, "[1] Amplitude Min/Max"=<0.1 0.5>, "[1] Bias"=1)
     )
     B=&Multiply_Float(
      A=&FloatFromCurve(FloatCurve=DataInterface() {propertyValues {Curve {keys [<at 0.9323 1>,<at 1 0>]}}})
      B=&Waveform("Waveform [1]"=::NewEnumerator0, "[1] Frequency"=23.7845, "[1] Amplitude Min/Max"=<0 2.5>, "[1] Bias"=1)
     )
     Alpha=$Particles.InterpolateSpeedThresholdEmitter_01.InterpolateOverTime_Float.Moving Average
    )
    ScaleA=true
    ScaleRGB=true
   )
  ]
  rendererClasses [&NiagaraSpriteRendererProperties, &NiagaraMeshRendererProperties]
  renderers [
   {rendererClass=&NiagaraSpriteRendererProperties}
   {rendererClass=&NiagaraMeshRendererProperties, rendererIndex=1}
  ]
 }
}
summary {
 systemName=NS_Emote_TangyRadishFlame_Sparks
 emitters [
  {emitterName=Flash_Flare, rendererClasses [&NiagaraSpriteRendererProperties, &NiagaraLightRendererProperties]}
  {emitterName=Classic_Sparks, rendererClasses [&NiagaraSpriteRendererProperties]}
  {emitterName=ParticleDecal, rendererClasses [&NiagaraSpriteRendererProperties, &NiagaraMeshRendererProperties]}
 ]
 userVariables [
  User.bIsFrontEnd {type=NiagaraBool}
  User.bIsFrontEndPreview {type=NiagaraBool}
  User.ForwardDot {type=NiagaraFloat, defaultValue=1, description=""}
  User.LerpCurve {type=NiagaraFloat, defaultValue=0, description=""}
 ]
}
